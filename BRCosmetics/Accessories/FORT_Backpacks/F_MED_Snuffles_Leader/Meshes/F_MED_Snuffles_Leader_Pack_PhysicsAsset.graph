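# Physics asset for the Snuffles Leader backpack cosmetic, giving its plush-like character rig simulated secondary motion.
# Simple box, capsule and sphere bodies cover the pack root, torso, neck, head, body, arms, feet and three scarf segments.
# The torso hangs off the pack bone and branches to neck-head, body-feet, arms and a chained scarf.
# Angular limits are tight on head and neck, looser on arms, scarf and torso-to-pack joint.
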PhysicsAsset /BRCosmetics/Accessories/FORT_Backpacks/F_MED_Snuffles_Leader/Meshes/F_MED_Snuffles_Leader_Pack_PhysicsAsset
bodies {
 dyn_arm_l [{shapeType=Box, extentX=12, extentY=3, extentZ=6, center=<-5 0 0>}]
 dyn_arm_r [{shapeType=Box, extentX=12, extentY=3, extentZ=6, center=<5 0 0>}]
 dyn_body [{shapeType=Capsule, length=1.8012, radius=8.9988, center=<7 0 0>, rotation=<rot 90 0 0>}]
 dyn_foot_l [{shapeType=Capsule, length=2, radius=2, center=<-1 0 0>, rotation=<rot 90 0 0>}]
 dyn_foot_r [{shapeType=Capsule, length=2, radius=2, center=<1 0 0>, rotation=<rot 90 0 0>}]
 dyn_head [{shapeType=Capsule, length=0, radius=6, center=<1 0 0>, rotation=<rot 90 0 0>}]
 dyn_neck [{shapeType=Capsule, length=0, radius=6, center=<1 0 0>, rotation=<rot 90 0 0>}]
 dyn_scarf_01 [{shapeType=Box, extentX=6, extentY=6, extentZ=1.2, center=<3 0 0.4>, rotation=<rot 5 0 0>}]
 dyn_scarf_02 [{shapeType=Box, extentX=6, extentY=6, extentZ=1.2, center=<3 0 0.1>}]
 dyn_scarf_03 [{shapeType=Box, extentX=4, extentY=6, extentZ=1.2, center=<2 0 0>}]
 dyn_torso [{shapeType=Sphere, radius=1}]
 pack [{shapeType=Box, extentX=1.01, extentY=1.01, extentZ=1.01, rotation=<rot 0 -90 0>}]
}
constraints [
 {bone1Name=dyn_torso, bone2Name=pack, swing1LimitDegrees=45, swing2LimitDegrees=45, twistLimitDegrees=45}
 {bone1Name=dyn_neck, bone2Name=dyn_torso, swing1LimitDegrees=8, swing2LimitDegrees=8, twistLimitDegrees=5}
 {bone1Name=dyn_head, bone2Name=dyn_neck, swing1LimitDegrees=8, swing2LimitDegrees=8, twistLimitDegrees=5}
 {bone1Name=dyn_body, bone2Name=dyn_torso, swing1LimitDegrees=15, swing2LimitDegrees=15, twistLimitDegrees=5}
 {bone1Name=dyn_foot_r, bone2Name=dyn_body, swing1LimitDegrees=15, swing2LimitDegrees=30, twistLimitDegrees=2}
 {bone1Name=dyn_foot_l, bone2Name=dyn_body, swing1LimitDegrees=15, swing2LimitDegrees=30, twistLimitDegrees=2}
 {bone1Name=dyn_scarf_01, bone2Name=dyn_torso, swing1LimitDegrees=30, swing2LimitDegrees=45, twistLimitDegrees=5}
 {bone1Name=dyn_scarf_02, bone2Name=dyn_scarf_01, swing1LimitDegrees=30, swing2LimitDegrees=45, twistLimitDegrees=5}
 {bone1Name=dyn_scarf_03, bone2Name=dyn_scarf_02, swing1LimitDegrees=30, swing2LimitDegrees=45, twistLimitDegrees=5}
 {bone1Name=dyn_arm_r, bone2Name=dyn_torso, swing1LimitDegrees=30, swing2LimitDegrees=30, twistLimitDegrees=5}
 {bone1Name=dyn_arm_l, bone2Name=dyn_torso, swing1LimitDegrees=30, swing2LimitDegrees=30, twistLimitDegrees=5}
]
